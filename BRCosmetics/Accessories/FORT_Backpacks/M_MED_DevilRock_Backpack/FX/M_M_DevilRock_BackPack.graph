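# Additive unlit, two-sided particle material for the DevilRock backpack cosmetic's FX.
# Use_Glow selects a ring glow from two sphere masks or a dual-panning noise color gradient, tinted by particle color and scaled by time of day.
# Use_sprite selects rune-masked, camera-facing mesh sprites expanded from UV1 with noise drift, or SplineThicken lines rotating around the particle position.
# Dynamic parameters drive line thickness and rotation speed; Use_WOP toggles the vertex offset.
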Material /BRCosmetics/Accessories/FORT_Backpacks/M_MED_DevilRock_Backpack/FX/M_M_DevilRock_BackPack
graph {
 properties {
  bUsedWithMeshParticles=true
  bUsedWithNiagaraSprites=true
  bUsedWithParticleSprites=true
  blendMode=BLEND_Additive
  pixelDepthOffsetMode=PDOM_Legacy
  shadingModel=MSM_Unlit
  twoSided=true
  usedShadingModels=Unlit
 }
 nodes [
  n1 = Constant() {r=0.5}
  n2 = Constant() {r=1}
  n3 = ScalarParameter() {defaultValue=1, parameterName=LerpHigh}
  n4 = TextureCoordinate()
  n5 = Constant() {r=0.5}
  n6 = Clamp(None=ScalarParameter() {defaultValue=0.46, parameterName=OutRadius}) {maxDefault=0.5}
  n7 = ScalarParameter() {defaultValue=0.8, parameterName=Hardness}
  n8 = TextureCoordinate()
  n9 = Constant() {r=0.5}
  n10 = Clamp(None=ScalarParameter() {defaultValue=0.45, parameterName=InnerRadius}) {maxDefault=0.5}
  n11 = Multiply(
   A=Add(A=SphereMask(A=@n4, B=@n5, Radius=@n6, Hardness=@n7) {attenuationRadius=0.5}) {constB=0}
   B=OneMinus(None=SphereMask(A=@n8, B=@n9, Radius=@n10, Hardness=@n7) {attenuationRadius=0.4})
  )
  n12 = Multiply(A=Power(Base=LinearInterpolate(B=@n3, Alpha=@n11), Exp=ScalarParameter() {defaultValue=2, parameterName=GlowPower}), B=ParticleColor().RGB)
  n13 = VectorParameter() {parameterName=Dark, defaultValue=<rgba 0.739 0.067 0 1>}
  n14 = VectorParameter() {parameterName=Bright, defaultValue=<rgba 1 0.5793 0 1>}
  n15 = VectorParameter() {defaultValue=<rgba 0.3 3 0.02 0.04>}
  n16 = Panner(Speed=AppendVector(A=@n15.B, B=@n15.B))
  n17 = AppendVector(A=@n15.R, B=@n15.R)
  n18 = VectorParameter() {parameterName=Param2, defaultValue=<rgba 0.3 2 0.04 0.02>}
  n19 = Panner(Speed=AppendVector(A=@n18.B, B=@n18.B))
  n20 = AppendVector(A=@n18.R, B=@n18.R)
  n21 = Multiply(
   A=TextureSample(UVs=Multiply(A=@n16, B=@n17)) {texture=&T_NoisePack_01}.G
   B=TextureSample(UVs=Multiply(A=@n19, B=@n20)) {texture=&T_NoisePack_01}.R
  )
  n22 = Multiply(A=@n21, B=Constant() {r=10})
  n23 = Constant() {r=0.2}
  n24 = ParticleColor()
  n25 = Multiply(
   A=LinearInterpolate(
    A=LinearInterpolate(A=@n13.RGB, B=@n14.RGB, Alpha=@n22)
    B=VectorParameter() {parameterName=White, defaultValue=<rgba 2 2 2 1>}.RGB
    Alpha=MaterialFunctionCall(In=@n21, Contrast=@n23) {materialFunction=&CheapContrast}.Result
   )
   B=@n24.RGB
  )
  MaterialFunctionCall(Emissive=Multiply(
    A=MaterialFunctionCall(Value_Morning=@n1, Value_Day=@n2, Value_Evening=@n1, Value_Night=@n1) {materialFunction=&Scalar_Time_of_Day}.Result
    B=StaticSwitchParameter(True=@n12, False=@n25) {parameterName=Use_Glow}
   )) {materialFunction=&MF_FX_EmissiveAdjustments}
  n26 = TextureCoordinate()
  Multiply(
   A=@n24.A
   B=StaticSwitchParameter(
    True=TextureSample(UVs=@n26) {texture=&/T_DevilRock_BackPack_Rune}.R
    False=Constant() {r=1}
   ) {parameterName=Use_sprite}
  )
  n27 = VertexColor()
  n28 = Multiply(
   A=Subtract(A=TextureCoordinate() {coordinateIndex=1, uTiling=2, vTiling=2})
   B=Multiply(A=ScalarParameter() {defaultValue=5, parameterName=Sprite_Size}, B=@n27.R)
  )
  n29 = Constant3Vector() {constant=<rgba 1 0 0 1>}
  n30 = Constant3Vector() {constant=<rgba 0 1 0 1>}
  n31 = Add(
   A=Multiply(A=ComponentMask(None=@n28) {r=true}, B=Transform(None=@n29) {transformSourceType=TRANSFORMSOURCE_Camera})
   B=Multiply(A=ComponentMask(None=@n28) {g=true}, B=Transform(None=@n30) {transformSourceType=TRANSFORMSOURCE_Camera})
  )
  n32 = Multiply(A=Time(), B=ScalarParameter() {defaultValue=0.02, parameterName=FloatingSpeed}) {constB=0.02}
  n33 = Multiply(A=TextureSample(UVs=Add(A=@n27.G, B=@n32)) {samplerType=SAMPLERTYPE_LinearColor, texture=&T_NoisePack_LinearColor}.RGB) {constB=3}
  n34 = DynamicParameter() {paramNames [Thickness, RotationSpeed, Param3, Param4]}
  n35 = Multiply(A=ScalarParameter() {defaultValue=1.5, parameterName=Thickness}, B=@n34.Thickness)
  n36 = MaterialFunctionCall(WidthBase=@n35, WidthTip=@n35) {materialFunction=&SplineThicken}
  n37 = RotateAboutAxis(
   NormalizedRotationAxis=Transform(None=VectorParameter() {parameterName=Rotation_axis, defaultValue=<rgba 0 1 0 1>}.RGB) {transformSourceType=TRANSFORMSOURCE_Local}
   RotationAngle=Multiply(A=Time(), B=@n34.RotationSpeed) {constB=-0.1}
   PivotPoint=ParticlePositionWS()
   Position=WorldPosition().XYZ
  )
  StaticSwitchParameter(
   True=StaticSwitchParameter(True=Add(A=@n31, B=@n33), False=Add(A=@n36.WorldPositionOffset, B=@n37)) {parameterName=Use_sprite}
   False=Constant()
  ) {defaultValue=true, parameterName=Use_WOP}
 ]
}
